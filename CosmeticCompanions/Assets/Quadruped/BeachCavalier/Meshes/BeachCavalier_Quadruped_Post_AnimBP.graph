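# Post-process AnimBlueprint for the BeachCavalier quadruped companion, adding pose-asset correctives and secondary motion to tail, ears and accessories.
# A pose blend feeds chained AnimDynamics tail simulation, cached for two bool-blended branches.
# One simulates full ears; the other constrains attach_emote_02 to weighted body bones, pins left ear bases and simulates ear tips and visor.
# Updates track outfit variant, spine bend and emote/gameplay state; output sets the AnimateAcc curve.
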AnimBlueprint /CosmeticCompanions/Assets/Quadruped/BeachCavalier/Meshes/BeachCavalier_Quadruped_Post_AnimBP
target_skeleton=&/BeachCavalier_Quadruped_Skeleton
graphs [
 AnimGraph {nodes [
   Output Pose(Result=Use cached pose 'AccCP'())
   Save cached pose 'inputCP'(Pose=Pose!(
     PoseAsset:Pose Asset Object Reference=Get PoseAsset() -> (PoseAsset:Pose Asset Object Reference)
     SourcePose:Pose Link Structure=Input Pose()
    ) -> (Pose:Pose Link Structure) {is=title AnimGraphNode_PoseBlendNode, bIgnorePoseAssetSourceAnimationCurves=true, blendOption=Linear}) {cacheName=inputCP}
   n1 = Anim Dynamics - Bone: dyn_tail_02!(ComponentPose=Local To Component(LocalPose=Use cached pose 'inputCP'()), Alpha=1) -> (Pose) {
    angularDampingOverride=0.9999
    angularSpringConstant=200
    bAngularSpring=true
    bOverrideAngularDamping=true
    bUseGravityOverride=true
    simulationSpace=RootRelative
    alphaBoolBlend {blendOption=Linear}
    boundBone {boneName=dyn_tail_02}
    componentAppliedLinearAccClamp=<3000 3000 3000>
    gravityOverride {z=-980}
    physicsBodyDefinitions [{
      boundBone {boneName=dyn_tail_02}
      boxExtents=<10 10 10>
      constraintSetup {angularLimitsMax=<0 10 20>, angularLimitsMin=<0 -10 -20>, angularTarget=<1 0 0>}
      localJointOffset=<4 0 0>
     }]
   }
   n2 = Anim Dynamics - Bone: dyn_tail_03!(ComponentPose=@n1, Alpha=1) -> (Pose) {
    angularDampingOverride=0.9999
    angularSpringConstant=200
    bAngularSpring=true
    bOverrideAngularDamping=true
    bUseGravityOverride=true
    simulationSpace=RootRelative
    alphaBoolBlend {blendOption=Linear}
    boundBone {boneName=dyn_tail_03}
    componentAppliedLinearAccClamp=<3000 3000 3000>
    gravityOverride {z=-980}
    physicsBodyDefinitions [{
      boundBone {boneName=dyn_tail_03}
      boxExtents=<10 10 10>
      constraintSetup {angularLimitsMax=<0 10 20>, angularLimitsMin=<0 -10 -20>, angularTarget=<1 0 0>}
      localJointOffset=<5 0 0>
     }]
   }
   n3 = Anim Dynamics - Bone: dyn_tail_04!(ComponentPose=@n2, Alpha=1) -> (Pose) {
    angularDampingOverride=0.9999
    angularSpringConstant=200
    bAngularSpring=true
    bOverrideAngularDamping=true
    bUseGravityOverride=true
    simulationSpace=RootRelative
    alphaBoolBlend {blendOption=Linear}
    boundBone {boneName=dyn_tail_04}
    componentAppliedLinearAccClamp=<3000 3000 3000>
    gravityOverride {z=-980}
    physicsBodyDefinitions [{
      boundBone {boneName=dyn_tail_04}
      boxExtents=<10 10 10>
      constraintSetup {angularLimitsMax=<0 10 20>, angularLimitsMin=<0 -10 -20>, angularTarget=<1 0 0>}
      localJointOffset=<6 0 0>
     }]
   }
   Save cached pose 'BodyDynCP'(Pose=Component To Local(ComponentPose=@n3)) {cacheName=BodyDynCP}
   n4 = Anim Dynamics - Bone: ear_02_l!(ComponentPose=Local To Component(LocalPose=Use cached pose 'BodyDynCP'()), Alpha=1) -> (Pose) {
    angularDampingOverride=0.9999
    angularSpringConstant=200
    bAngularSpring=true
    bOverrideAngularDamping=true
    bUseGravityOverride=true
    simulationSpace=RootRelative
    alphaBoolBlend {blendOption=Linear}
    boundBone {boneName=ear_02_l}
    componentAppliedLinearAccClamp=<3000 3000 3000>
    gravityOverride {z=-980}
    physicsBodyDefinitions [{
      boundBone {boneName=ear_02_l}
      boxExtents=<10 10 10>
      constraintSetup {angularLimitsMax=<0 5 0>, angularLimitsMin=<0 -5 -10>, angularTarget=<1 0 0>}
      localJointOffset=<2 0 0>
     }]
   }
   n5 = Anim Dynamics - Bone: ear_03_l!(ComponentPose=@n4, Alpha=1) -> (Pose) {
    angularDampingOverride=0.9999
    angularSpringConstant=200
    bAngularSpring=true
    bOverrideAngularDamping=true
    bUseGravityOverride=true
    simulationSpace=RootRelative
    alphaBoolBlend {blendOption=Linear}
    boundBone {boneName=ear_03_l}
    componentAppliedLinearAccClamp=<3000 3000 3000>
    gravityOverride {z=-980}
    physicsBodyDefinitions [{
      boundBone {boneName=ear_03_l}
      boxExtents=<10 10 10>
      constraintSetup {angularLimitsMax=<0 5 0>, angularLimitsMin=<0 -5 -10>, angularTarget=<1 0 0>}
      localJointOffset=<2 0 0>
     }]
   }
   n6 = Anim Dynamics - Bone: ear_04_l!(ComponentPose=@n5, Alpha=1) -> (Pose) {
    angularDampingOverride=0.9999
    angularSpringConstant=200
    bAngularSpring=true
    bOverrideAngularDamping=true
    bUseGravityOverride=true
    simulationSpace=RootRelative
    alphaBoolBlend {blendOption=Linear}
    boundBone {boneName=ear_04_l}
    componentAppliedLinearAccClamp=<3000 3000 3000>
    gravityOverride {z=-980}
    physicsBodyDefinitions [{
      boundBone {boneName=ear_04_l}
      boxExtents=<10 10 10>
      constraintSetup {angularLimitsMax=<0 5 0>, angularLimitsMin=<0 -5 -10>, angularTarget=<1 0 0>}
      localJointOffset=<2 0 0>
     }]
   }
   n7 = Anim Dynamics - Bone: ear_02_r!(ComponentPose=@n6, Alpha=1) -> (Pose) {
    angularDampingOverride=0.9999
    angularSpringConstant=200
    bAngularSpring=true
    bOverrideAngularDamping=true
    bUseGravityOverride=true
    simulationSpace=RootRelative
    alphaBoolBlend {blendOption=Linear}
    boundBone {boneName=ear_02_r}
    componentAppliedLinearAccClamp=<3000 3000 3000>
    gravityOverride {z=-980}
    physicsBodyDefinitions [{
      boundBone {boneName=ear_02_r}
      boxExtents=<10 10 10>
      constraintSetup {angularLimitsMax=<0 5 0>, angularLimitsMin=<0 -5 -10>, angularTarget=<1 0 0>}
      localJointOffset=<-2 0 0>
     }]
   }
   n8 = Anim Dynamics - Bone: ear_03_r!(ComponentPose=@n7, Alpha=1) -> (Pose) {
    angularDampingOverride=0.9999
    angularSpringConstant=200
    bAngularSpring=true
    bOverrideAngularDamping=true
    bUseGravityOverride=true
    simulationSpace=RootRelative
    alphaBoolBlend {blendOption=Linear}
    boundBone {boneName=ear_03_r}
    componentAppliedLinearAccClamp=<3000 3000 3000>
    gravityOverride {z=-980}
    physicsBodyDefinitions [{
      boundBone {boneName=ear_03_r}
      boxExtents=<10 10 10>
      constraintSetup {angularLimitsMax=<0 5 0>, angularLimitsMin=<0 -5 -10>, angularTarget=<1 0 0>}
      localJointOffset=<-2 0 0>
     }]
   }
   n9 = Anim Dynamics - Bone: ear_04_r!(ComponentPose=@n8, Alpha=1) -> (Pose) {
    angularDampingOverride=0.9999
    angularSpringConstant=200
    bAngularSpring=true
    bOverrideAngularDamping=true
    bUseGravityOverride=true
    simulationSpace=RootRelative
    alphaBoolBlend {blendOption=Linear}
    boundBone {boneName=ear_04_r}
    componentAppliedLinearAccClamp=<3000 3000 3000>
    gravityOverride {z=-980}
    physicsBodyDefinitions [{
      boundBone {boneName=ear_04_r}
      boxExtents=<10 10 10>
      constraintSetup {angularLimitsMax=<0 5 0>, angularLimitsMin=<0 -5 -10>, angularTarget=<1 0 0>}
      localJointOffset=<-2 0 0>
     }]
   }
   n10 = Component To Local(ComponentPose=@n9)
   n11 = "Constraint - attach_emote_02 (PPPPPPP)"!(
    ComponentPose=Local To Component(LocalPose=Use cached pose 'BodyDynCP'())
    Alpha=1
    ConstraintWeights_0=0.4923
    ConstraintWeights_1:float=0.2367
    ConstraintWeights_2:float=0.232
    ConstraintWeights_3:float=0.0198
    ConstraintWeights_4:float=0.0162
    ConstraintWeights_5:float=0.0026
    ConstraintWeights_6:float=0.0003
   ) -> (Pose) {
    boneToModify {boneName=attach_emote_02}
    constraintSetup [
     {targetBone {boneName=spine_03}}
     {targetBone {boneName=spine_01}}
     {targetBone {boneName=spine_02}}
     {targetBone {boneName=neck_01}}
     {targetBone {boneName=shoulder_fr_l}}
     {targetBone {boneName=pelvis}}
     {targetBone {boneName=neck_02}}
    ]
    constraintWeights [1,1,1,1,1,1,1]
   }
   n12 = Component To Local(ComponentPose=@n11)
   n13 = "Constraint - attach_emote_02 (PPPPPPP)"(
    ComponentPose=Local To Component(LocalPose=Use cached pose 'BodyDynCP'())
    Alpha=1
    ConstraintWeights_0=0.4923
    ConstraintWeights_1=0.2367
    ConstraintWeights_2=0.232
    ConstraintWeights_3=0.0198
    ConstraintWeights_4=0.0162
    ConstraintWeights_5=0.0026
    ConstraintWeights_6=0.0003
   ) {
    boneToModify {boneName=attach_emote_02}
    constraintSetup [
     {targetBone {boneName=spine_03}}
     {targetBone {boneName=spine_01}}
     {targetBone {boneName=spine_02}}
     {targetBone {boneName=neck_01}}
     {targetBone {boneName=shoulder_fr_l}}
     {targetBone {boneName=pelvis}}
     {targetBone {boneName=neck_02}}
    ]
    constraintWeights [1,1,1,1,1,1,1]
   }
   n14 = Component To Local(ComponentPose=@n13)
   n15 = "Transform (Modify) Bone - Bone: attach_emote_02"(
    ComponentPose=Local To Component(LocalPose=Blend Poses by bool(BlendPose_0=@n12, BlendPose_1=@n14))
    Alpha=1
    Translation="2.636,3.260,0.952"
    Rotation="1.526,32.428,6.077"
    Scale="1,1,1"
   ) {
    rotationMode=BMM_Additive
    translationMode=BMM_Additive
    boneToModify {boneName=attach_emote_02}
    rotation=<rot 1.526 32.428 6.077>
    translation=<2.636 3.26 0.952>
   }
   n16 = "Transform (Modify) Bone - Bone: ear_01_l"!(ComponentPose=@n15, Alpha=1, Translation="0,0,0", Rotation="0,0,0") -> (Pose) {
    rotationMode=BMM_Replace
    rotationSpace=BCS_BoneSpace
    translationMode=BMM_Replace
    translationSpace=BCS_BoneSpace
    boneToModify {boneName=ear_01_l}
   }
   n17 = "Transform (Modify) Bone - Bone: ear_02_l"!(ComponentPose=@n16, Alpha=1, Translation="0,0,0", Rotation="0,0,0") -> (Pose) {
    rotationMode=BMM_Replace
    rotationSpace=BCS_BoneSpace
    translationMode=BMM_Replace
    translationSpace=BCS_BoneSpace
    boneToModify {boneName=ear_02_l}
   }
   n18 = "Transform (Modify) Bone - Bone: ear_01_l"(ComponentPose=@n17, Alpha=1, Translation="0,0,0", Rotation="0,0,0") {
    rotationMode=BMM_Replace
    rotationSpace=BCS_BoneSpace
    translationMode=BMM_Replace
    translationSpace=BCS_BoneSpace
    boneToModify {boneName=ear_01_l}
   }
   n19 = "Transform (Modify) Bone - Bone: ear_02_l"(ComponentPose=@n18, Alpha=1, Translation="0,0,0", Rotation="0,0,0") {
    rotationMode=BMM_Replace
    rotationSpace=BCS_BoneSpace
    translationMode=BMM_Replace
    translationSpace=BCS_BoneSpace
    boneToModify {boneName=ear_02_l}
   }
   n20 = Anim Dynamics - Bone: ear_03_l(ComponentPose=@n19, Alpha=1) {
    angularDampingOverride=0.9999
    angularSpringConstant=200
    bAngularSpring=true
    bOverrideAngularDamping=true
    bUseGravityOverride=true
    simulationSpace=RootRelative
    alphaBoolBlend {blendOption=Linear}
    boundBone {boneName=ear_03_l}
    componentAppliedLinearAccClamp=<3000 3000 3000>
    gravityOverride {z=-980}
    physicsBodyDefinitions [{
      boundBone {boneName=ear_03_l}
      boxExtents=<10 10 10>
      constraintSetup {angularLimitsMax=<0 5 0>, angularLimitsMin=<0 -5 -12>, angularTarget=<1 0 0>}
      localJointOffset=<2 0 0>
     }]
   }
   n21 = Anim Dynamics - Bone: ear_04_l(ComponentPose=@n20, Alpha=1) {
    angularDampingOverride=0.9999
    angularSpringConstant=200
    bAngularSpring=true
    bOverrideAngularDamping=true
    bUseGravityOverride=true
    simulationSpace=RootRelative
    alphaBoolBlend {blendOption=Linear}
    boundBone {boneName=ear_04_l}
    componentAppliedLinearAccClamp=<3000 3000 3000>
    gravityOverride {z=-980}
    physicsBodyDefinitions [{
      boundBone {boneName=ear_04_l}
      boxExtents=<10 10 10>
      constraintSetup {angularLimitsMax=<0 5 0>, angularLimitsMin=<0 -5 -10>, angularTarget=<1 0 0>}
      localJointOffset=<2 0 0>
     }]
   }
   n22 = Anim Dynamics - Bone: ear_03_r(ComponentPose=@n21, Alpha=1) {
    angularDampingOverride=0.9999
    angularSpringConstant=200
    bAngularSpring=true
    bOverrideAngularDamping=true
    bUseGravityOverride=true
    simulationSpace=RootRelative
    alphaBoolBlend {blendOption=Linear}
    boundBone {boneName=ear_03_r}
    componentAppliedLinearAccClamp=<3000 3000 3000>
    gravityOverride {z=-980}
    physicsBodyDefinitions [{
      boundBone {boneName=ear_03_r}
      boxExtents=<10 10 10>
      constraintSetup {angularLimitsMax=<0 5 0>, angularLimitsMin=<0 -5 -12>, angularTarget=<1 0 0>}
      localJointOffset=<-2 0 0>
     }]
   }
   n23 = Anim Dynamics - Bone: ear_04_r(ComponentPose=@n22, Alpha=1) {
    angularDampingOverride=0.9999
    angularSpringConstant=200
    bAngularSpring=true
    bOverrideAngularDamping=true
    bUseGravityOverride=true
    simulationSpace=RootRelative
    alphaBoolBlend {blendOption=Linear}
    boundBone {boneName=ear_04_r}
    componentAppliedLinearAccClamp=<3000 3000 3000>
    gravityOverride {z=-980}
    physicsBodyDefinitions [{
      boundBone {boneName=ear_04_r}
      boxExtents=<10 10 10>
      constraintSetup {angularLimitsMax=<0 5 0>, angularLimitsMin=<0 -5 -10>, angularTarget=<1 0 0>}
      localJointOffset=<-2 0 0>
     }]
   }
   n24 = Anim Dynamics - Bone: dyn_visor!(ComponentPose=@n23, Alpha=1, GravityOverride:Vector="0,0,-980") -> (Pose) {
    angularDampingOverride=1
    angularSpringConstant=3000
    bAngularSpring=true
    bOverrideAngularDamping=true
    bUseGravityOverride=true
    gravityScale=0.35
    simulationSpace=RootRelative
    alphaBoolBlend {blendOption=Linear}
    boundBone {boneName=dyn_visor}
    componentAppliedLinearAccClamp=<10000 10000 10000>
    componentLinearAccScale=<1 1 1>
    componentLinearVelScale=<1 1 1>
    gravityOverride {z=-980}
    physicsBodyDefinitions [{
      boundBone {boneName=dyn_visor}
      boxExtents=<25 30 15>
      constraintSetup {angularLimitsMax=<0 10 0>, angularLimitsMin=<0 -10 0>, angularTarget=<1 0 0>}
      localJointOffset=<4 0 0>
     }]
    simSpaceSettings {maxAngularAcceleration=30, maxAngularVelocity=6, simSpaceAngularAlpha=1}
   }
   n25 = Component To Local(ComponentPose=@n24)
   Save cached pose 'AccCP'(Pose=Modify Curve(SourcePose=Blend Poses by bool(BlendPose_0=@n10, BlendPose_1=@n25), CurveValues_0=Get AnimateAcc() -> (AnimateAcc:double), Alpha=1) {curveNames [AnimateAcc], curveValues [0]}) {cacheName=AccCP}
  ]}
 EventGraph {nodes [
   Event BlueprintInitializeAnimation()
   n1 = GetOwningActor()
   n2 = GetOwningActor()
   > Set isAnimatingGameplay(isAnimatingGameplay:bool=OR Boolean(
     A=IsA!(Object:Object Reference=@n2, ObjectClass:Object Class Reference=/CosmeticCompanions/Assets/Shared/Blueprint/BP_Companion.BP_Companion_C) -> (ReturnValue:bool)
     B=IsA(Object=@n1, ObjectClass=/Script/CosmeticCompanionsRuntime.FortCompanionMannequin)
    )) -> (Output_Get:bool)
   Event BlueprintUpdateAnimation()
   > n3 = Sequence()
   > n4 = Set isHelmetUnequipped(isHelmetUnequipped:bool=HasVariantTagForChannel!(
     self:Fort Player Companion Anim Instance Object Reference=GetMainAnimInstance!(self:Companion Post Process Anim Instance Object Reference) -> (ReturnValue:Fort Player Companion Anim Instance Object Reference)
     ChannelTag:Gameplay Tag Structure=(TagName="Cosmetics.Variant.Channel.Outfit")
     VariantTag:Gameplay Tag Structure=(TagName="Cosmetics.Variant.Property.Stage1")
    ) -> (ReturnValue:bool)) -> (Output_Get:bool)
   > n5 = Branch(Condition=@n4)
   Set SpineBendFloat(
    execute=@n5.else
    SpineBendFloat:float=GetAlphaFromReferenceTransformDelta(SkeletalMeshComponent=GetOwningComponent(), BoneName=spine_02, BaseName=spine_01, InputRange=(X=0,Y=-15), TwistAxis="0,0,1")
   ) -> (Output_Get:float)
   Set EmoteOrAnimating(
    execute=@n3.then_1
    EmoteOrAnimating:bool=OR Boolean(A=Get bUseReverseRetarget() -> (bUseReverseRetarget:bool), B=Get isAnimatingGameplay() -> (isAnimatingGameplay:bool))
   ) -> (Output_Get:bool)
  ]}
]
variables [
 AnimateAcc {type=number, default=1}
 EmoteOrAnimating {type=boolean}
 isAnimatingGameplay {type=boolean}
 isHelmetUnequipped {type=boolean}
 PoseAsset {type=PoseAsset, default=&/BeachCavalier_Quadruped_PoseAsset}
 ShovelConstraint {type=number}
 SpineBendFloat {type=number}
]
